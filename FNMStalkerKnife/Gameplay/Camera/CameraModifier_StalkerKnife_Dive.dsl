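; Camera modifier for the Stalker Knife dive, supplying a steeply downward target rotation relative to the view target.
; GetTargetCameraRotation rotates the view target's forward vector around its right vector by a fixed angle, then builds a rotator from that direction.
; GetRelativePitchFrustum only passes through the parent's value.
; How the rotation is applied or blended is not defined here.
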(fn GetRelativePitchFrustum
  (bind _outrelativepitchfrustum (|CallParentFunction))
  (return _outrelativepitchfrustum))

(fn GetTargetCameraRotation
  (return (Math|Rotator|MakeRotfromX (Math|Vector|RotateVectorAroundAxis (Transformation|GetActorForwardVector (CameraModifier|GetViewTarget)) 65.0 (Transformation|GetActorRightVector (CameraModifier|GetViewTarget))))))
